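# Conversation database for the Tropical Medic specialist NPC, defining what it says when a player starts talking to it.
# The entry point leads to one speech node, which holds a default greeting and contextual lines gated by controller requirements.
# Curve-driven priorities cover first-ever meetings and reactions to specific outfits, while other lines check the per-match conversation count.
# The speech then links out to the shared NPC quests and basic services conversations.
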ConversationDatabase /NPCLibrary/NPCs/Tropical/NPCLibrary_Conversation_Tropical
entry_points [{tag=Conversation.Entry.Tropical, destinations [#1]}]
nodes [
 ConversationEntryPointNode_7 {
  id=#1
  class=ConversationEntryPointNode
  outputs [#2]
  properties {entryTag {tagName=Conversation.Entry.Tropical}}
 }
 Speech_C_4 {
  id=#2
  class=Speech_C
  outputs [#3, #4]
  properties {
   nodeName=""
   generalConfig {
    defaultMessage=Need any help out there?
    contextualMessages [
     {
      message="Oh hey! You look, uh... like you could use a hand."
      contextRequirements [{requirement {class=FortControllerRequirement_IsFirstEverConversationWithNPC}}]
      priority {curve {rowName=Tandem.Speech.Priority.FirstEverInteraction}}
     }
     {
      message=I hear you host amazing beach parties. Awesome.
      contextRequirements [{requirement {class=FortControllerRequirement_HasCID, properties {allowedCIds [/BRCosmetics/Athena/Items/Cosmetics/Characters/Character_TideKing_Regal]}}}]
      priority {curve {rowName=Tandem.Speech.Priority.OutfitReaction}}
     }
     {
      message="Who's a good pup?!"
      contextRequirements [{requirement {
         class=FortControllerRequirement_HasCID
         properties {allowedCIds [/BRCosmetics/Athena/Items/Cosmetics/Characters/Character_GateHound_Spectral]}
        }}]
      priority {curve {rowName=Tandem.Speech.Priority.OutfitReaction}}
     }
     {
      message=Need any help out there?
      contextRequirements [{requirement {
         class=FortControllerRequirement_CheckNPCMatchConversationCount
         properties {nPCUniqueIdTag {tagName=AISpawnerData.Type.Specialist.Medic.Tropical}}
        }}]
      priority {curve {rowName=None}}
     }
     {
      message=Staying safe? I hear the waves are getting choppy.
      contextRequirements [{requirement {
         class=FortControllerRequirement_CheckNPCMatchConversationCount
         properties {nPCUniqueIdTag {tagName=AISpawnerData.Type.Specialist.Medic.Tropical}}
        }}]
      priority {curve {rowName=None}}
     }
     {
      message="Don't forget your Slurpscreen!"
      contextRequirements [{requirement {
         class=FortControllerRequirement_CheckNPCMatchConversationCount
         properties {nPCUniqueIdTag {tagName=AISpawnerData.Type.Specialist.Medic.Tropical}}
        }}]
      priority {curve {rowName=None}}
     }
    ]
   }
  }
 }
 ConversationLinkNode_0 {id=#4, class=ConversationLinkNode, properties {remoteEntryTag {tagName=Conversation.Entry.BasicServices}}}
 ConversationLinkNode_1 {id=#3, class=ConversationLinkNode, properties {remoteEntryTag {tagName=Conversation.Entry.NPCLibrary.Quests}}}
]
